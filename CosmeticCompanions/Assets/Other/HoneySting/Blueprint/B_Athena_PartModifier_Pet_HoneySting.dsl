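; Cosmetic part modifier Blueprint for the HoneySting pet companion that keeps its idle particle effect matched to the wing flap animation.
; After the parent's post-res-in logic, it runs once to get the part's skeletal mesh on the player pawn and cache its HoneySting AnimBP.
; A looping 0.1-second timer then copies the AnimBP's WingFlapAlpha into the same-named float parameter on the idle VFX component.
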(event EventOnPostResIn
  ;[0.8783 0.3489 0.2094 0.5] Match with the AnimBP
  (|CallParentFunction)
  (Utilities|FlowControl|DoOnce false
    (:Completed
      (bind _asskeletal_mesh_component (Utilities|Casting|CastToSkeletalMeshComponent (CustomCharacter|GetSkeletalMeshForPartOnPlayerPawn)))
      (bind _ashoney_sting_other_anim_bp (Utilities|Casting|CastToHoneySting_Other_AnimBP (Components|SkeletalMesh|GetAnimInstance _asskeletal_mesh_component)))
      (Variables|Default|SetAnimBP _ashoney_sting_other_anim_bp)
      (Utilities|Time|SetTimerbyEvent (delegate Custom|Update%20WingAlpha) 0.1 true))))

(event Custom|Update%20WingAlpha
  ;[0.8783 0.3489 0.2094 0.5] Match with the AnimBP
  (Effects|Components|ParticleSystem|SetFloatParameter (CallFunction|GetIdleVFXComponent) "WingFlapAlpha" (Variables|/CosmeticCompanions/Assets/Other/HoneySting/Meshes/HoneySting_Other_AnimBP.HoneySting_Other_AnimBP_C|GetWingFlapAlpha (Variables|Default|GetAnimBP))))

(fn ConstructionScript
  (|CallParentFunction))
