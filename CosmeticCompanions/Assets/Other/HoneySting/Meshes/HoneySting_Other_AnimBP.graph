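# Animation Blueprint for the HoneySting cosmetic companion's "Other" skeleton, driving it with retargeted poses and a wing flap layer.
# A bUseReverseRetarget bool switches between two Retarget Pose From Mesh nodes, selecting the emote IK Retargeter when true and the standard one otherwise.
# An additive wing flap sequence is applied on top, weighted by WingFlapAlpha.
# The Copy Pose From Mesh node is unconnected, and the event graph holds only unwired stubs.
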AnimBlueprint /CosmeticCompanions/Assets/Other/HoneySting/Meshes/HoneySting_Other_AnimBP
target_skeleton=&/HoneySting_Other_Skeleton
graphs [
 AnimGraph {nodes [
   Copy Pose From Mesh() {bCopyCurves=true, bUseAttachedParent=true}
   n1 = Retarget Pose From Mesh!() -> (Pose:Pose Link Structure) {is=title AnimGraphNode_RetargetPoseFromMesh, iKRetargeterAsset=&/RTG_HoneySting_Emote}
   n2 = Retarget Pose From Mesh() {iKRetargeterAsset=&/RTG_HoneySting}
   n3 = Get bUseReverseRetarget() -> (bUseReverseRetarget:bool)
   Output Pose(Result=Apply Additive(
     Base=Blend Poses by bool(BlendPose_0=@n1, BlendPose_1=@n2, bActiveValue=@n3)
     Additive="Sequence Player (additive) 'HoneySting_ACT_WingFlapA'"() {sequence=/CosmeticCompanions/Assets/Other/HoneySting/Animation/Actions/HoneySting_ACT_WingFlapA}
     Alpha=Get WingFlapAlpha() -> (WingFlapAlpha:double)
    ))
  ]}
 EventGraph {nodes [
   TryGetPawnOwner()
   Event BlueprintUpdateAnimation()
  ]}
]
variables [WingFlapAlpha {type=number, default=1}]
