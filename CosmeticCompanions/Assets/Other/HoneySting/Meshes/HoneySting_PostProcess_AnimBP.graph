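# Post-process AnimBP for the HoneySting cosmetic companion, adding facial poses and secondary motion.
# Input runs through a Pose Asset, then a bool blend between raw and constrained poses, where fin bases, leg bases and head center follow weighted spine/head bones.
# AnimDynamics then moves spine, legs, antennae and hair, and springs move droplets, before output.
# Hat dynamics and top-fin rotation constraints are unconnected; EventGraph sets gameplay and top-wing variables.
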AnimBlueprint /CosmeticCompanions/Assets/Other/HoneySting/Meshes/HoneySting_PostProcess_AnimBP
target_skeleton=&/HoneySting_Other_Skeleton
graphs [
 AnimGraph {nodes [
   Anim Dynamics - Bone: hat(Alpha=1, ExternalForce="0,0,0") {
    angularDampingOverride=1
    angularSpringConstant=10000
    bAngularSpring=true
    bOverrideAngularDamping=true
    bUseGravityOverride=true
    gravityScale=0.35
    simulationSpace=BoneRelative
    alphaScaleBiasClamp {bias=1, scale=-1}
    boundBone {boneName=hat}
    componentAppliedLinearAccClamp=<1000 1000 1000>
    componentLinearAccScale=<1 1 1>
    componentLinearVelScale=<1 1 1>
    gravityOverride {z=-50}
    physicsBodyDefinitions [{
      boundBone {boneName=hat}
      boxExtents=<25 25 20>
      constraintSetup {angularTargetAxis=AxisZ, angularLimitsMax=<20 25 5>, angularLimitsMin=<-20 -25 -5>, angularTarget=<0 0 1>}
      localJointOffset=<0 0 -3>
     }]
    relativeSpaceBone {boneName=fish_01}
    simSpaceSettings {maxAngularAcceleration=25, maxAngularVelocity=5, simSpaceAngularAlpha=1}
   }
   n1 = "Constraint - fintop_r_01 (R)"(Alpha=1, ConstraintWeights_0=1) {
    alphaScaleBiasClamp {bias=1, scale=-1}
    boneToModify {boneName=fintop_r_01}
    constraintSetup [{transformType=Rotation, targetBone {boneName=fin_r_01}}]
    constraintWeights [1]
   }
   n2 = "Constraint - fintop_r_02 (R)"(ComponentPose=@n1, Alpha=1, ConstraintWeights_0=1) {
    alphaScaleBiasClamp {bias=1, scale=-1}
    boneToModify {boneName=fintop_r_02}
    constraintSetup [{transformType=Rotation, targetBone {boneName=fin_r_02}}]
    constraintWeights [1]
   }
   n3 = "Constraint - fintop_r_03 (R)"(ComponentPose=@n2, Alpha=1, ConstraintWeights_0=1) {
    alphaScaleBiasClamp {bias=1, scale=-1}
    boneToModify {boneName=fintop_r_03}
    constraintSetup [{transformType=Rotation, targetBone {boneName=fin_r_01}}]
    constraintWeights [1]
   }
   n4 = "Constraint - fintop_l_01 (R)"(ComponentPose=@n3, Alpha=1, ConstraintWeights_0=1) {
    alphaScaleBiasClamp {bias=1, scale=-1}
    boneToModify {boneName=fintop_l_01}
    constraintSetup [{transformType=Rotation, targetBone {boneName=fin_l_01}}]
    constraintWeights [1]
   }
   n5 = "Constraint - fintop_l_02 (R)"(ComponentPose=@n4, Alpha=1, ConstraintWeights_0=1) {
    alphaScaleBiasClamp {bias=1, scale=-1}
    boneToModify {boneName=fintop_l_02}
    constraintSetup [{transformType=Rotation, targetBone {boneName=fin_l_02}}]
    constraintWeights [1]
   }
   "Constraint - fintop_l_03 (R)"(ComponentPose=@n5, Alpha=1, ConstraintWeights_0=1) {
    alphaScaleBiasClamp {bias=1, scale=-1}
    boneToModify {boneName=fintop_l_03}
    constraintSetup [{transformType=Rotation, targetBone {boneName=fin_l_03}}]
    constraintWeights [1]
   }
   Output Pose(Result=Use cached pose 'DynamicsCache'())
   Save cached pose 'Cache_Input'(Pose=Pose!(
     PoseAsset:Pose Asset Object Reference=Get PoseAsset() -> (PoseAsset:Pose Asset Object Reference)
     SourcePose:Pose Link Structure=Input Pose()
    ) -> (Pose:Pose Link Structure) {is=title AnimGraphNode_PoseBlendNode, bIgnorePoseAssetSourceAnimationCurves=true, blendOption=Linear}) {cacheName=Cache_Input}
   n6 = Anim Dynamics - Bone: spine_01(ComponentPose=Local To Component(LocalPose=Use cached pose 'preDynCache'()), Alpha=1, ExternalForce="0,0,0") {
    angularDampingOverride=1
    angularSpringConstant=10000
    bAngularSpring=true
    bOverrideAngularDamping=true
    bUseGravityOverride=true
    gravityScale=0.35
    simulationSpace=BoneRelative
    alphaScaleBiasClamp {bias=1, scale=-1}
    boundBone {boneName=spine_01}
    componentAppliedLinearAccClamp=<1000 1000 1000>
    componentLinearAccScale=<1 1 1>
    componentLinearVelScale=<1 1 1>
    physicsBodyDefinitions [{
      boundBone {boneName=spine_01}
      boxExtents=<25 35 35>
      constraintSetup {angularLimitsMax=<30 20 15>, angularLimitsMin=<-30 -20 -10>, angularTarget=<1 0 0>}
      localJointOffset=<2 0 0>
     }]
    relativeSpaceBone {boneName=fish_01}
    simSpaceSettings {maxAngularAcceleration=24, maxAngularVelocity=4, simSpaceAngularAlpha=1}
   }
   n7 = Anim Dynamics - Bone: spine_02(ComponentPose=@n6, Alpha=1, ExternalForce="0,0,0") {
    angularDampingOverride=1
    angularSpringConstant=10000
    bAngularSpring=true
    bOverrideAngularDamping=true
    bUseGravityOverride=true
    gravityScale=0.35
    simulationSpace=BoneRelative
    alphaScaleBiasClamp {bias=1, scale=-1}
    boundBone {boneName=spine_02}
    componentAppliedLinearAccClamp=<1000 1000 1000>
    componentLinearAccScale=<1 1 1>
    componentLinearVelScale=<1 1 1>
    physicsBodyDefinitions [{
      boundBone {boneName=spine_02}
      boxExtents=<25 35 35>
      constraintSetup {angularLimitsMax=<30 15 10>, angularLimitsMin=<-30 -15 -10>, angularTarget=<1 0 0>}
      localJointOffset=<1.5 0 0>
     }]
    relativeSpaceBone {boneName=fish_01}
    simSpaceSettings {maxAngularAcceleration=24, maxAngularVelocity=4, simSpaceAngularAlpha=1}
   }
   n8 = Anim Dynamics - Bone: spine_03(ComponentPose=@n7, Alpha=1, ExternalForce="0,0,0") {
    angularDampingOverride=1
    angularSpringConstant=10000
    bAngularSpring=true
    bOverrideAngularDamping=true
    bUseGravityOverride=true
    gravityScale=0.35
    simulationSpace=BoneRelative
    alphaScaleBiasClamp {bias=1, scale=-1}
    boundBone {boneName=spine_03}
    componentAppliedLinearAccClamp=<1000 1000 1000>
    componentLinearAccScale=<1 1 1>
    componentLinearVelScale=<1 1 1>
    physicsBodyDefinitions [{
      boundBone {boneName=spine_03}
      boxExtents=<25 35 35>
      constraintSetup {angularLimitsMax=<30 20 20>, angularLimitsMin=<-30 -20 -10>, angularTarget=<1 0 0>}
      localJointOffset=<2 0 0>
     }]
    relativeSpaceBone {boneName=fish_01}
    simSpaceSettings {maxAngularAcceleration=24, maxAngularVelocity=4, simSpaceAngularAlpha=1}
   }
   n9 = Anim Dynamics - Bone: leg_A_l_01(ComponentPose=@n8, Alpha=1, ExternalForce="0,0,0") {
    angularDampingOverride=1
    angularSpringConstant=2000
    bOverrideAngularDamping=true
    bUseGravityOverride=true
    gravityScale=0.35
    simulationSpace=BoneRelative
    alphaScaleBiasClamp {bias=1, scale=-1}
    boundBone {boneName=leg_A_l_01}
    componentAppliedLinearAccClamp=<1000 1000 1000>
    componentLinearAccScale=<1 1 1>
    componentLinearVelScale=<1 1 1>
    gravityOverride {z=-980}
    physicsBodyDefinitions [{
      boundBone {boneName=leg_A_l_01}
      boxExtents=<10 15 15>
      constraintSetup {angularLimitsMax=<15 40 15>, angularLimitsMin=<-15 -40 -45>, angularTarget=<1 0 0>}
      localJointOffset=<2.5 0 0>
     }]
    relativeSpaceBone {boneName=fish_01}
    simSpaceSettings {maxAngularAcceleration=20, maxAngularVelocity=4, simSpaceAngularAlpha=1}
   }
   n10 = Anim Dynamics - Bone: leg_B_l_01(ComponentPose=@n9, Alpha=1, ExternalForce="0,0,0") {
    angularDampingOverride=1
    angularSpringConstant=2000
    bOverrideAngularDamping=true
    bUseGravityOverride=true
    gravityScale=0.35
    simulationSpace=BoneRelative
    alphaScaleBiasClamp {bias=1, scale=-1}
    boundBone {boneName=leg_B_l_01}
    componentAppliedLinearAccClamp=<1000 1000 1000>
    componentLinearAccScale=<1 1 1>
    componentLinearVelScale=<1 1 1>
    gravityOverride {z=-980}
    physicsBodyDefinitions [{
      boundBone {boneName=leg_B_l_01}
      boxExtents=<10 15 15>
      constraintSetup {angularLimitsMax=<15 40 15>, angularLimitsMin=<-15 -25 -45>, angularTarget=<1 0 0>}
      localJointOffset=<2.5 0 0>
     }]
    relativeSpaceBone {boneName=fish_01}
    simSpaceSettings {maxAngularAcceleration=20, maxAngularVelocity=4, simSpaceAngularAlpha=1}
   }
   n11 = Anim Dynamics - Bone: leg_C_l_01(ComponentPose=@n10, Alpha=1, ExternalForce="0,0,0") {
    angularDampingOverride=1
    angularSpringConstant=2000
    bOverrideAngularDamping=true
    bUseGravityOverride=true
    gravityScale=0.35
    simulationSpace=BoneRelative
    alphaScaleBiasClamp {bias=1, scale=-1}
    boundBone {boneName=leg_C_l_01}
    componentAppliedLinearAccClamp=<1000 1000 1000>
    componentLinearAccScale=<1 1 1>
    componentLinearVelScale=<1 1 1>
    gravityOverride {z=-980}
    physicsBodyDefinitions [{
      boundBone {boneName=leg_C_l_01}
      boxExtents=<10 15 15>
      constraintSetup {angularLimitsMax=<15 40 15>, angularLimitsMin=<-15 -40 -45>, angularTarget=<1 0 0>}
      localJointOffset=<2.5 0 0>
     }]
    relativeSpaceBone {boneName=fish_01}
    simSpaceSettings {maxAngularAcceleration=20, maxAngularVelocity=4, simSpaceAngularAlpha=1}
   }
   n12 = Anim Dynamics - Bone: leg_A_r_01(ComponentPose=@n11, Alpha=1, ExternalForce="0,0,0") {
    angularDampingOverride=1
    angularSpringConstant=2000
    bOverrideAngularDamping=true
    bUseGravityOverride=true
    gravityScale=0.35
    simulationSpace=BoneRelative
    alphaScaleBiasClamp {bias=1, scale=-1}
    boundBone {boneName=leg_A_r_01}
    componentAppliedLinearAccClamp=<1000 1000 1000>
    componentLinearAccScale=<1 1 1>
    componentLinearVelScale=<1 1 1>
    gravityOverride {z=-980}
    physicsBodyDefinitions [{
      boundBone {boneName=leg_A_r_01}
      boxExtents=<10 15 15>
      constraintSetup {angularLimitsMax=<15 45 0>, angularLimitsMin=<-15 -45 -45>, angularTarget=<1 0 0>}
      localJointOffset=<-2.5 0 0>
     }]
    relativeSpaceBone {boneName=fish_01}
    simSpaceSettings {maxAngularAcceleration=20, maxAngularVelocity=4, simSpaceAngularAlpha=1}
   }
   n13 = Anim Dynamics - Bone: leg_B_r_01(ComponentPose=@n12, Alpha=1, ExternalForce="0,0,0") {
    angularDampingOverride=1
    angularSpringConstant=2000
    bOverrideAngularDamping=true
    bUseGravityOverride=true
    gravityScale=0.35
    simulationSpace=BoneRelative
    alphaScaleBiasClamp {bias=1, scale=-1}
    boundBone {boneName=leg_B_r_01}
    componentAppliedLinearAccClamp=<1000 1000 1000>
    componentLinearAccScale=<1 1 1>
    componentLinearVelScale=<1 1 1>
    gravityOverride {z=-980}
    physicsBodyDefinitions [{
      boundBone {boneName=leg_B_r_01}
      boxExtents=<10 15 15>
      constraintSetup {angularLimitsMax=<15 45 0>, angularLimitsMin=<-15 -25 -45>, angularTarget=<1 0 0>}
      localJointOffset=<-2.5 0 0>
     }]
    relativeSpaceBone {boneName=fish_01}
    simSpaceSettings {maxAngularAcceleration=20, maxAngularVelocity=4, simSpaceAngularAlpha=1}
   }
   n14 = Anim Dynamics - Bone: leg_C_r_01(ComponentPose=@n13, Alpha=1, ExternalForce="0,0,0") {
    angularDampingOverride=1
    angularSpringConstant=2000
    bOverrideAngularDamping=true
    bUseGravityOverride=true
    gravityScale=0.35
    simulationSpace=BoneRelative
    alphaScaleBiasClamp {bias=1, scale=-1}
    boundBone {boneName=leg_C_r_01}
    componentAppliedLinearAccClamp=<1000 1000 1000>
    componentLinearAccScale=<1 1 1>
    componentLinearVelScale=<1 1 1>
    gravityOverride {z=-980}
    physicsBodyDefinitions [{
      boundBone {boneName=leg_C_r_01}
      boxExtents=<10 15 15>
      constraintSetup {angularLimitsMax=<15 45 0>, angularLimitsMin=<-15 -45 -45>, angularTarget=<1 0 0>}
      localJointOffset=<-2.5 0 0>
     }]
    relativeSpaceBone {boneName=fish_01}
    simSpaceSettings {maxAngularAcceleration=20, maxAngularVelocity=4, simSpaceAngularAlpha=1}
   }
   n15 = Anim Dynamics - Bone: antennae_l_01(ComponentPose=@n14, Alpha=1, ExternalForce="0,0,0") {
    angularDampingOverride=1
    angularSpringConstant=5000
    bAngularSpring=true
    bOverrideAngularDamping=true
    bUseGravityOverride=true
    gravityScale=0.35
    simulationSpace=BoneRelative
    alphaScaleBiasClamp {bias=1, scale=-1}
    boundBone {boneName=antennae_l_01}
    componentAppliedLinearAccClamp=<1000 1000 1000>
    componentLinearAccScale=<1 1 1>
    componentLinearVelScale=<1 1 1>
    physicsBodyDefinitions [{
      boundBone {boneName=antennae_l_01}
      boxExtents=<12 18 18>
      constraintSetup {angularLimitsMax=<15 60 20>, angularLimitsMin=<-15 -15 -20>, angularTarget=<1 0 0>}
      localJointOffset=<-2 0 0>
     }]
    relativeSpaceBone {boneName=fish_01}
    simSpaceSettings {maxAngularAcceleration=25, maxAngularVelocity=5, simSpaceAngularAlpha=1}
   }
   n16 = Anim Dynamics - Bone: antennae_l_02(ComponentPose=@n15, Alpha=1, ExternalForce="0,0,0") {
    angularDampingOverride=1
    angularSpringConstant=5000
    bAngularSpring=true
    bOverrideAngularDamping=true
    bUseGravityOverride=true
    gravityScale=0.35
    simulationSpace=BoneRelative
    alphaScaleBiasClamp {bias=1, scale=-1}
    boundBone {boneName=antennae_l_02}
    componentAppliedLinearAccClamp=<1000 1000 1000>
    componentLinearAccScale=<1 1 1>
    componentLinearVelScale=<1 1 1>
    physicsBodyDefinitions [{
      boundBone {boneName=antennae_l_02}
      boxExtents=<10 15 15>
      constraintSetup {angularLimitsMax=<0 60 30>, angularLimitsMin=<0 -15 -30>, angularTarget=<1 0 0>}
      localJointOffset=<1 0 0>
     }]
    relativeSpaceBone {boneName=fish_01}
    simSpaceSettings {maxAngularAcceleration=25, maxAngularVelocity=5, simSpaceAngularAlpha=1}
   }
   n17 = Anim Dynamics - Bone: antennae_r_01(ComponentPose=@n16, Alpha=1, ExternalForce="0,0,0") {
    angularDampingOverride=1
    angularSpringConstant=5000
    bAngularSpring=true
    bOverrideAngularDamping=true
    bUseGravityOverride=true
    gravityScale=0.35
    simulationSpace=BoneRelative
    alphaScaleBiasClamp {bias=1, scale=-1}
    boundBone {boneName=antennae_r_01}
    componentAppliedLinearAccClamp=<1000 1000 1000>
    componentLinearAccScale=<1 1 1>
    componentLinearVelScale=<1 1 1>
    physicsBodyDefinitions [{
      boundBone {boneName=antennae_r_01}
      boxExtents=<12 18 18>
      constraintSetup {angularLimitsMax=<15 60 20>, angularLimitsMin=<-15 -15 -20>, angularTarget=<1 0 0>}
      localJointOffset=<-2 0 0>
     }]
    relativeSpaceBone {boneName=fish_01}
    simSpaceSettings {maxAngularAcceleration=25, maxAngularVelocity=5, simSpaceAngularAlpha=1}
   }
   n18 = Anim Dynamics - Bone: antennae_r_02(ComponentPose=@n17, Alpha=1, ExternalForce="0,0,0") {
    angularDampingOverride=1
    angularSpringConstant=5000
    bAngularSpring=true
    bOverrideAngularDamping=true
    bUseGravityOverride=true
    gravityScale=0.35
    simulationSpace=BoneRelative
    alphaScaleBiasClamp {bias=1, scale=-1}
    boundBone {boneName=antennae_r_02}
    componentAppliedLinearAccClamp=<1000 1000 1000>
    componentLinearAccScale=<1 1 1>
    componentLinearVelScale=<1 1 1>
    physicsBodyDefinitions [{
      boundBone {boneName=antennae_r_02}
      boxExtents=<10 15 15>
      constraintSetup {angularLimitsMax=<0 60 30>, angularLimitsMin=<0 -15 -30>, angularTarget=<1 0 0>}
      localJointOffset=<-1 0 0>
     }]
    relativeSpaceBone {boneName=fish_01}
    simSpaceSettings {maxAngularAcceleration=25, maxAngularVelocity=5, simSpaceAngularAlpha=1}
   }
   n19 = Anim Dynamics - Bone: hair_01(ComponentPose=@n18, Alpha=1, ExternalForce="0,0,0") {
    angularDampingOverride=1
    angularSpringConstant=10000
    bAngularSpring=true
    bOverrideAngularDamping=true
    bUseGravityOverride=true
    gravityScale=0.35
    simulationSpace=BoneRelative
    alphaScaleBiasClamp {bias=1, scale=-1}
    boundBone {boneName=hair_01}
    componentAppliedLinearAccClamp=<1000 1000 1000>
    componentLinearAccScale=<1 1 1>
    componentLinearVelScale=<1 1 1>
    gravityOverride {z=-100}
    physicsBodyDefinitions [{
      boundBone {boneName=hair_01}
      boxExtents=<20 30 30>
      constraintSetup {angularLimitsMax=<25 25 30>, angularLimitsMin=<-25 -25 -30>, angularTarget=<1 0 0>}
      localJointOffset=<4.5 0 0>
     }]
    relativeSpaceBone {boneName=fish_01}
    simSpaceSettings {maxAngularAcceleration=25, maxAngularVelocity=5, simSpaceAngularAlpha=1}
   }
   n20 = Anim Dynamics - Bone: hair_02(ComponentPose=@n19, Alpha=1, ExternalForce="0,0,0") {
    angularDampingOverride=1
    angularSpringConstant=10000
    bAngularSpring=true
    bOverrideAngularDamping=true
    bUseGravityOverride=true
    gravityScale=0.35
    simulationSpace=BoneRelative
    alphaScaleBiasClamp {bias=1, scale=-1}
    boundBone {boneName=hair_02}
    componentAppliedLinearAccClamp=<1000 1000 1000>
    componentLinearAccScale=<1 1 1>
    componentLinearVelScale=<1 1 1>
    gravityOverride {z=-100}
    physicsBodyDefinitions [{
      boundBone {boneName=hair_02}
      boxExtents=<20 30 30>
      constraintSetup {angularLimitsMax=<25 25 30>, angularLimitsMin=<-25 -25 -30>, angularTarget=<1 0 0>}
      localJointOffset=<4.5 0 0>
     }]
    relativeSpaceBone {boneName=fish_01}
    simSpaceSettings {maxAngularAcceleration=25, maxAngularVelocity=5, simSpaceAngularAlpha=1}
   }
   n21 = Spring controller - Bone: droplet_l!(ComponentPose:Component Space Pose Link Structure=@n20, Alpha:float=1) -> (Pose:Component Space Pose Link Structure) {
    is=title AnimGraphNode_SpringBone
    bLimitDisplacement=true
    bOverrideOwnerVelocity=true
    maxDisplacement=0.75
    springDamping=75
    springStiffness=300
    springBone {boneName=droplet_l}
   }
   n22 = Spring controller - Bone: droplet_m!(ComponentPose:Component Space Pose Link Structure=@n21, Alpha:float=1) -> (Pose:Component Space Pose Link Structure) {
    is=title AnimGraphNode_SpringBone
    bLimitDisplacement=true
    bOverrideOwnerVelocity=true
    maxDisplacement=0.75
    springDamping=75
    springStiffness=300
    springBone {boneName=droplet_m}
   }
   n23 = Spring controller - Bone: droplet_r!(ComponentPose:Component Space Pose Link Structure=@n22, Alpha:float=1) -> (Pose:Component Space Pose Link Structure) {
    is=title AnimGraphNode_SpringBone
    bLimitDisplacement=true
    bOverrideOwnerVelocity=true
    maxDisplacement=0.75
    springDamping=75
    springStiffness=300
    springBone {boneName=droplet_r}
   }
   Save cached pose 'DynamicsCache'(Pose=Component To Local(ComponentPose=@n23)) {cacheName=DynamicsCache}
   n24 = "Constraint - fintop_base_l_01 (PPP)"!(
    ComponentPose=Local To Component(LocalPose=Use cached pose 'Cache_Input'())
    Alpha=1
    ConstraintWeights_0=0.5737
    ConstraintWeights_1:float=0.3675
    ConstraintWeights_2:float=0.0589
   ) -> (Pose) {
    boneToModify {boneName=fintop_base_l_01}
    constraintSetup [{targetBone {boneName=spine_01}}, {targetBone {boneName=spine_02}}, {targetBone {boneName=spine_03}}]
    constraintWeights [1,1,1]
   }
   n25 = "Constraint - fin_base_l_01 (PPP)"!(ComponentPose=@n24, Alpha=1, ConstraintWeights_0=0.4942, ConstraintWeights_1:float=0.4176, ConstraintWeights_2:float=0.0882) -> (Pose) {
    boneToModify {boneName=fin_base_l_01}
    constraintSetup [{targetBone {boneName=spine_01}}, {targetBone {boneName=spine_02}}, {targetBone {boneName=spine_03}}]
    constraintWeights [1,1,1]
   }
   n26 = "Constraint - fintop_base_r_01 (PPP)"!(ComponentPose=@n25, Alpha=1, ConstraintWeights_0=0.5737, ConstraintWeights_1:float=0.3675, ConstraintWeights_2:float=0.0589) -> (Pose) {
    boneToModify {boneName=fintop_base_r_01}
    constraintSetup [{targetBone {boneName=spine_01}}, {targetBone {boneName=spine_02}}, {targetBone {boneName=spine_03}}]
    constraintWeights [1,1,1]
   }
   n27 = "Constraint - fin_base_r_01 (PPP)"!(ComponentPose=@n26, Alpha=1, ConstraintWeights_0=0.4942, ConstraintWeights_1:float=0.4176, ConstraintWeights_2:float=0.0882) -> (Pose) {
    boneToModify {boneName=fin_base_r_01}
    constraintSetup [{targetBone {boneName=spine_01}}, {targetBone {boneName=spine_02}}, {targetBone {boneName=spine_03}}]
    constraintWeights [1,1,1]
   }
   n28 = "Constraint - leg_Base_A_l_01 (PPPP)"!(
    ComponentPose=@n27
    Alpha=1
    ConstraintWeights_0=0.5238
    ConstraintWeights_1:float=0.278
    ConstraintWeights_2:float=0.1709
    ConstraintWeights_3:float=0.0273
   ) -> (Pose) {
    boneToModify {boneName=leg_Base_A_l_01}
    constraintSetup [{targetBone {boneName=spine_02}}, {targetBone {boneName=spine_01}}, {targetBone {boneName=spine_03}}, {targetBone {boneName=head_01}}]
    constraintWeights [1,1,1,1]
   }
   n29 = "Constraint - leg_Base_B_l_01 (PP)"!(ComponentPose=@n28, Alpha=1, ConstraintWeights_0=0.9769, ConstraintWeights_1:float=0.0231) -> (Pose) {
    boneToModify {boneName=leg_Base_B_l_01}
    constraintSetup [{targetBone {boneName=spine_03}}, {targetBone {boneName=spine_02}}]
    constraintWeights [1,1]
   }
   n30 = "Constraint - leg_Base_C_l_01 (P)"!(ComponentPose=@n29, Alpha=1, ConstraintWeights_0=1, ConstraintWeights_1:float=0) -> (Pose) {boneToModify {boneName=leg_Base_C_l_01}, constraintSetup [{targetBone {boneName=spine_03}}], constraintWeights [1,0]}
   n31 = "Constraint - leg_Base_A_r_01 (PPP)"!(ComponentPose=@n30, Alpha=1, ConstraintWeights_0=0.4921, ConstraintWeights_1:float=0.2787, ConstraintWeights_2:float=0.2291) -> (Pose) {
    boneToModify {boneName=leg_Base_A_r_01}
    constraintSetup [{targetBone {boneName=spine_02}}, {targetBone {boneName=spine_01}}, {targetBone {boneName=spine_03}}]
    constraintWeights [1,1,1]
   }
   n32 = "Constraint - leg_Base_B_r_01 (PP)"!(ComponentPose=@n31, Alpha=1, ConstraintWeights_0=0.9766, ConstraintWeights_1:float=0.0234) -> (Pose) {
    boneToModify {boneName=leg_Base_B_r_01}
    constraintSetup [{targetBone {boneName=spine_03}}, {targetBone {boneName=spine_02}}]
    constraintWeights [1,1]
   }
   n33 = "Constraint - leg_Base_C_r_01 (P)"!(ComponentPose=@n32, Alpha=1, ConstraintWeights_0=1, ConstraintWeights_1:float=0) -> (Pose) {boneToModify {boneName=leg_Base_C_r_01}, constraintSetup [{targetBone {boneName=spine_03}}], constraintWeights [1,0]}
   n34 = "Constraint - hat_headCenter (RR)"!(ComponentPose=@n33, Alpha=1, ConstraintWeights_0=0.5, ConstraintWeights_1:float=0.5) -> (Pose) {
    alphaScaleBiasClamp {bias=1, scale=-1}
    boneToModify {boneName=hat_headCenter}
    constraintSetup [{transformType=Rotation, targetBone {boneName=head_01}}, {transformType=Rotation, targetBone {boneName=spine_01}}]
    constraintWeights [1,1]
   }
   Save cached pose 'preDynCache'(Pose=Blend Poses by bool(BlendPose_0=Use cached pose 'Cache_Input'(), BlendPose_1=Component To Local(ComponentPose=@n34))) {cacheName=preDynCache}
  ]}
 EventGraph {nodes [
   Event BlueprintUpdateAnimation()
   > Set disableTopWings(disableTopWings:float=GetCurveValue(CurveName=disable_top_wings)) -> (Output_Get:float)
   Event BlueprintInitializeAnimation()
   n1 = GetOwningActor()
   n2 = GetOwningActor()
   > Set isAnimatingGameplay(isAnimatingGameplay:bool=OR Boolean(
     A=IsA!(Object:Object Reference=@n2, ObjectClass:Object Class Reference=/CosmeticCompanions/Assets/Shared/Blueprint/BP_Companion.BP_Companion_C) -> (ReturnValue:bool)
     B=IsA(Object=@n1, ObjectClass=/Script/CosmeticCompanionsRuntime.FortCompanionMannequin)
    )) -> (Output_Get:bool)
   TryGetPawnOwner()
  ]}
]
variables [
 disableTopWings {type=number}
 isAnimatingGameplay {type=boolean}
 PoseAsset {type=PoseAsset, default=&/HoneySting_Other_Head_Facial_Poses_PoseAsset}
]
